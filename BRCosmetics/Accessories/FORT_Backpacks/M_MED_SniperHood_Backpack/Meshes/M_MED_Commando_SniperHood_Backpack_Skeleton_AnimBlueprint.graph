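# Animation Blueprint for the Commando SniperHood backpack cosmetic, adding secondary motion and body-driven rotation to an input pose.
# A chain of Anim Dynamics nodes simulates damped, angle-limited frill bones under reduced gravity, plus a tighter, spring-driven scope bone.
# Each update, the event graph remaps the owner mesh's spine_05-to-spine_01 delta-from-ref-pose yaw into a clamped backpack_rotate yaw.
# That rotation is applied additively to backpack_root in bone space after the dynamics.
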AnimBlueprint /BRCosmetics/Accessories/FORT_Backpacks/M_MED_SniperHood_Backpack/Meshes/M_MED_Commando_SniperHood_Backpack_Skeleton_AnimBlueprint
target_skeleton=&/M_MED_Commando_SniperHood_Backpack_Skeleton
graphs [
 AnimGraph {nodes [
   AnimGraphNode_SubInput_0 = Input Pose()
   n1 = Anim Dynamics - Bone: dyn_frillA_01(ComponentPose=Local To Component(LocalPose=@AnimGraphNode_SubInput_0), Alpha=0.65, ExternalForce="0,0,0") {
    angularDampingOverride=0.98
    bGravityOverrideInSimSpace=true
    bOverrideAngularDamping=true
    gravityScale=0.45
    alphaBoolBlend {blendOption=Linear}
    boundBone {boneName=dyn_frillA_01}
    componentAppliedLinearAccClamp=<1000 1000 1000>
    componentLinearAccScale=<1 1 1>
    componentLinearVelScale=<3 3 3>
    physicsBodyDefinitions [{
      sphereCollisionRadius=0
      boundBone {boneName=dyn_frillA_01}
      boxExtents=<30 30 30>
      constraintSetup {angularLimitsMax=<30 50 0>, angularLimitsMin=<-30 -50 0>, angularTarget=<0 0 0>}
      localJointOffset=<0 0 -8>
     }]
   }
   n2 = Anim Dynamics - Bone: dyn_frillA_02(ComponentPose=@n1, Alpha=0.75, ExternalForce="0,0,0") {
    angularDampingOverride=0.98
    bGravityOverrideInSimSpace=true
    bOverrideAngularDamping=true
    gravityScale=0.45
    alphaBoolBlend {blendOption=Linear}
    boundBone {boneName=dyn_frillA_02}
    componentAppliedLinearAccClamp=<1000 1000 1000>
    componentLinearAccScale=<1 1 1>
    componentLinearVelScale=<3 3 3>
    physicsBodyDefinitions [{
      sphereCollisionRadius=0
      boundBone {boneName=dyn_frillA_02}
      boxExtents=<30 30 30>
      constraintSetup {angularLimitsMax=<30 50 0>, angularLimitsMin=<-30 -50 0>, angularTarget=<0 0 0>}
      localJointOffset=<0 0 -8>
     }]
   }
   n3 = Anim Dynamics - Bone: dyn_frillB_01(ComponentPose=@n2, Alpha=0.75, ExternalForce="0,0,0") {
    angularDampingOverride=0.98
    bGravityOverrideInSimSpace=true
    bOverrideAngularDamping=true
    gravityScale=0.45
    alphaBoolBlend {blendOption=Linear}
    boundBone {boneName=dyn_frillB_01}
    componentAppliedLinearAccClamp=<1000 1000 1000>
    componentLinearAccScale=<1 1 1>
    componentLinearVelScale=<3 3 3>
    physicsBodyDefinitions [{
      sphereCollisionRadius=0
      boundBone {boneName=dyn_frillB_01}
      boxExtents=<30 30 30>
      constraintSetup {angularLimitsMax=<30 50 0>, angularLimitsMin=<-30 -50 0>, angularTarget=<0 0 0>}
      localJointOffset=<0 0 -8>
     }]
   }
   n4 = Anim Dynamics - Bone: dyn_frillC_01(ComponentPose=@n3, Alpha=0.55, ExternalForce="0,0,0") {
    angularDampingOverride=0.98
    bGravityOverrideInSimSpace=true
    bOverrideAngularDamping=true
    gravityScale=0.45
    alphaBoolBlend {blendOption=Linear}
    boundBone {boneName=dyn_frillC_01}
    componentAppliedLinearAccClamp=<1000 1000 1000>
    componentLinearAccScale=<1 1 1>
    componentLinearVelScale=<3 3 3>
    physicsBodyDefinitions [{
      sphereCollisionRadius=0
      boundBone {boneName=dyn_frillC_01}
      boxExtents=<30 30 30>
      constraintSetup {twistAxis=AxisZ, angularLimitsMax=<30 50 0>, angularLimitsMin=<-30 -50 0>, angularTarget=<0 0 0>}
      localJointOffset=<2.5 0 -8>
     }]
   }
   n5 = Anim Dynamics - Bone: dyn_frillC_02(ComponentPose=@n4, Alpha=0.65, ExternalForce="0,0,0") {
    angularDampingOverride=0.98
    bGravityOverrideInSimSpace=true
    bOverrideAngularDamping=true
    gravityScale=0.45
    alphaBoolBlend {blendOption=Linear}
    boundBone {boneName=dyn_frillC_02}
    componentAppliedLinearAccClamp=<1000 1000 1000>
    componentLinearAccScale=<1 1 1>
    componentLinearVelScale=<3 3 3>
    physicsBodyDefinitions [{
      sphereCollisionRadius=0
      boundBone {boneName=dyn_frillC_02}
      boxExtents=<30 30 30>
      constraintSetup {twistAxis=AxisZ, angularLimitsMax=<30 50 0>, angularLimitsMin=<-30 -50 0>, angularTarget=<0 0 0>}
      localJointOffset=<0 0 -8>
     }]
   }
   n6 = Anim Dynamics - Bone: dyn_frillC_03(ComponentPose=@n5, Alpha=0.9, ExternalForce="0,0,0") {
    angularDampingOverride=0.98
    bGravityOverrideInSimSpace=true
    bOverrideAngularDamping=true
    gravityScale=0.45
    alphaBoolBlend {blendOption=Linear}
    boundBone {boneName=dyn_frillC_03}
    componentAppliedLinearAccClamp=<1000 1000 1000>
    componentLinearAccScale=<1 1 1>
    componentLinearVelScale=<3 3 3>
    physicsBodyDefinitions [{
      sphereCollisionRadius=0
      boundBone {boneName=dyn_frillC_03}
      boxExtents=<30 30 30>
      constraintSetup {twistAxis=AxisZ, angularLimitsMax=<30 50 0>, angularLimitsMin=<-30 -50 0>, angularTarget=<0 0 0>}
      localJointOffset=<0 0 -8>
     }]
   }
   n7 = Anim Dynamics - Bone: dyn_frillD_01(ComponentPose=@n6, Alpha=0.5, ExternalForce="0,0,0") {
    angularDampingOverride=0.98
    bGravityOverrideInSimSpace=true
    bOverrideAngularDamping=true
    gravityScale=0.45
    alphaBoolBlend {blendOption=Linear}
    boundBone {boneName=dyn_frillD_01}
    componentAppliedLinearAccClamp=<1000 1000 1000>
    componentLinearAccScale=<1 1 1>
    componentLinearVelScale=<3 3 3>
    physicsBodyDefinitions [{
      sphereCollisionRadius=0
      boundBone {boneName=dyn_frillD_01}
      boxExtents=<30 30 30>
      constraintSetup {twistAxis=AxisZ, angularLimitsMax=<30 50 0>, angularLimitsMin=<-30 -50 0>, angularTarget=<0 0 0>}
      localJointOffset=<0 0 -8>
     }]
   }
   n8 = Anim Dynamics - Bone: dyn_frillD_02(ComponentPose=@n7, Alpha=0.8, ExternalForce="0,0,0") {
    angularDampingOverride=0.98
    bGravityOverrideInSimSpace=true
    bOverrideAngularDamping=true
    gravityScale=0.45
    alphaBoolBlend {blendOption=Linear}
    boundBone {boneName=dyn_frillD_02}
    componentAppliedLinearAccClamp=<1000 1000 1000>
    componentLinearAccScale=<1 1 1>
    componentLinearVelScale=<3 3 3>
    physicsBodyDefinitions [{
      sphereCollisionRadius=0
      boundBone {boneName=dyn_frillD_02}
      boxExtents=<30 30 30>
      constraintSetup {twistAxis=AxisZ, angularLimitsMax=<30 50 0>, angularLimitsMin=<-30 -50 0>, angularTarget=<0 0 0>}
      localJointOffset=<0 0 -8>
     }]
   }
   n9 = Anim Dynamics - Bone: dyn_frillE_01(ComponentPose=@n8, Alpha=0.55, ExternalForce="0,0,0") {
    angularDampingOverride=0.98
    bGravityOverrideInSimSpace=true
    bOverrideAngularDamping=true
    gravityScale=0.45
    alphaBoolBlend {blendOption=Linear}
    boundBone {boneName=dyn_frillE_01}
    componentAppliedLinearAccClamp=<1000 1000 1000>
    componentLinearAccScale=<1 1 1>
    componentLinearVelScale=<3 3 3>
    physicsBodyDefinitions [{
      sphereCollisionRadius=0
      boundBone {boneName=dyn_frillE_01}
      boxExtents=<30 30 30>
      constraintSetup {twistAxis=AxisZ, angularLimitsMax=<30 50 0>, angularLimitsMin=<-30 -50 0>, angularTarget=<0 0 0>}
      localJointOffset=<3 0 -8>
     }]
   }
   n10 = Anim Dynamics - Bone: dyn_frillE_02(ComponentPose=@n9, Alpha=0.75, ExternalForce="0,0,0") {
    angularDampingOverride=0.98
    bGravityOverrideInSimSpace=true
    bOverrideAngularDamping=true
    gravityScale=0.45
    alphaBoolBlend {blendOption=Linear}
    boundBone {boneName=dyn_frillE_02}
    componentAppliedLinearAccClamp=<1000 1000 1000>
    componentLinearAccScale=<1 1 1>
    componentLinearVelScale=<3 3 3>
    physicsBodyDefinitions [{
      sphereCollisionRadius=0
      boundBone {boneName=dyn_frillE_02}
      boxExtents=<30 30 30>
      constraintSetup {twistAxis=AxisZ, angularLimitsMax=<30 50 0>, angularLimitsMin=<-30 -50 0>, angularTarget=<0 0 0>}
      localJointOffset=<3 0 -8>
     }]
   }
   n11 = Anim Dynamics - Bone: dyn_frillF_01(ComponentPose=@n10, Alpha=0.55, ExternalForce="0,0,0") {
    angularDampingOverride=0.98
    bGravityOverrideInSimSpace=true
    bOverrideAngularDamping=true
    gravityScale=0.45
    alphaBoolBlend {blendOption=Linear}
    boundBone {boneName=dyn_frillF_01}
    componentAppliedLinearAccClamp=<1000 1000 1000>
    componentLinearAccScale=<1 1 1>
    componentLinearVelScale=<3 3 3>
    physicsBodyDefinitions [{
      sphereCollisionRadius=0
      boundBone {boneName=dyn_frillF_01}
      boxExtents=<30 30 30>
      constraintSetup {twistAxis=AxisZ, angularLimitsMax=<30 50 0>, angularLimitsMin=<-30 -50 0>, angularTarget=<0 0 0>}
      localJointOffset=<3 0 -8>
     }]
   }
   n12 = Anim Dynamics - Bone: dyn_frillF_02(ComponentPose=@n11, Alpha=0.7, ExternalForce="0,0,0") {
    angularDampingOverride=0.98
    bGravityOverrideInSimSpace=true
    bOverrideAngularDamping=true
    gravityScale=0.45
    alphaBoolBlend {blendOption=Linear}
    boundBone {boneName=dyn_frillF_02}
    componentAppliedLinearAccClamp=<1000 1000 1000>
    componentLinearAccScale=<1 1 1>
    componentLinearVelScale=<3 3 3>
    physicsBodyDefinitions [{
      sphereCollisionRadius=0
      boundBone {boneName=dyn_frillF_02}
      boxExtents=<30 30 30>
      constraintSetup {twistAxis=AxisZ, angularLimitsMax=<30 50 0>, angularLimitsMin=<-30 -50 0>, angularTarget=<0 0 0>}
      localJointOffset=<3 0 -8>
     }]
   }
   n13 = Anim Dynamics - Bone: dyn_frillG_01(ComponentPose=@n12, Alpha=0.75, ExternalForce="0,0,0") {
    angularDampingOverride=0.98
    bGravityOverrideInSimSpace=true
    bOverrideAngularDamping=true
    gravityScale=0.45
    alphaBoolBlend {blendOption=Linear}
    boundBone {boneName=dyn_frillG_01}
    componentAppliedLinearAccClamp=<1000 1000 1000>
    componentLinearAccScale=<1 1 1>
    componentLinearVelScale=<3 3 3>
    physicsBodyDefinitions [{
      sphereCollisionRadius=0
      boundBone {boneName=dyn_frillG_01}
      boxExtents=<30 30 30>
      constraintSetup {twistAxis=AxisZ, angularLimitsMax=<30 50 0>, angularLimitsMin=<-30 -50 0>, angularTarget=<0 0 0>}
      localJointOffset=<-2 0 -15>
     }]
   }
   n14 = Anim Dynamics - Bone: dyn_frillH_01(ComponentPose=@n13, Alpha=0.75, ExternalForce="0,0,0") {
    angularDampingOverride=0.98
    bGravityOverrideInSimSpace=true
    bOverrideAngularDamping=true
    gravityScale=0.45
    alphaBoolBlend {blendOption=Linear}
    boundBone {boneName=dyn_frillH_01}
    componentAppliedLinearAccClamp=<1000 1000 1000>
    componentLinearAccScale=<1 1 1>
    componentLinearVelScale=<3 3 3>
    physicsBodyDefinitions [{
      sphereCollisionRadius=0
      boundBone {boneName=dyn_frillH_01}
      boxExtents=<30 30 30>
      constraintSetup {twistAxis=AxisZ, angularLimitsMax=<30 50 0>, angularLimitsMin=<-30 -50 0>, angularTarget=<0 0 0>}
      localJointOffset=<-2 0 -8>
     }]
   }
   n15 = Anim Dynamics - Bone: dyn_scope(ComponentPose=@n14, Alpha=0.3, ExternalForce="0,0,0") {
    angularDampingOverride=0.98
    angularSpringConstant=5000
    bAngularSpring=true
    bGravityOverrideInSimSpace=true
    bOverrideAngularDamping=true
    alphaBoolBlend {blendOption=Linear}
    boundBone {boneName=dyn_scope}
    componentAppliedLinearAccClamp=<1000 1000 1000>
    componentLinearVelScale=<3 3 3>
    physicsBodyDefinitions [{
      sphereCollisionRadius=0
      boundBone {boneName=dyn_scope}
      boxExtents=<30 30 30>
      constraintSetup {angularLimitsMax=<5 15 0>, angularLimitsMin=<-3 -15 0>, angularTarget=<0 0 0>}
      localJointOffset=<0 0 -15>
     }]
   }
   n16 = "Transform (Modify) Bone - Bone: backpack_root"(ComponentPose=@n15, Alpha=1, Translation="0,0,0", Rotation=Get backpack_rotate(), Scale="1,1,1") {rotationMode=BMM_Additive, rotationSpace=BCS_BoneSpace, alphaBoolBlend {blendOption=Linear}, boneToModify {boneName=backpack_root}}
   Output Pose(Result=Component To Local(ComponentPose=@n16)) {layerGroup=None}
  ]}
 EventGraph {nodes [
   Event BlueprintUpdateAnimation()
   n1 = GetOwnerSkeletalMeshComponent()
   n2 = BreakRotator(InRot=BreakTransform(InTransform=GetDeltaTransformFromRefPose(self=@n1, BoneName=spine_05, BaseName=spine_01)).Rotation)
   > Set backpack_rotate(backpack_rotate=MakeRotator(Yaw=MapRangeClamped(Value=@n2.Yaw, InRangeA=-5, InRangeB=-30, OutRangeB=32)))
  ]}
]
variables [backpack_rotate {type=Rotator}]
